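# Material for the lowest-detail LOD of the Sparks crowd, drawn as instanced meshes or Niagara mesh particles.
# World position offset combines pivot-based playback of three animation textures, driven by time, position and per-instance random, with look-at rotation.
# An MPC CrowdResponse value selects animations.
# Quaternion-rotated axes crossed with camera vectors create cheap spline-thickened strips with synthetic normals, alongside curve-atlas colors and optional stage-colored emissive.
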Material /SparksMaps/Crowd/Materials/LastLOD/M_SparksCrowdLastLOD
graph {
 properties {
  bCastRayTracedShadows=false
  bTangentSpaceNormal=false
  bUsedWithInstancedStaticMeshes=true
  bUsedWithNiagaraMeshParticles=true
  pixelDepthOffsetMode=PDOM_Legacy
  usedShadingModels=Default Lit
 }
 nodes [
  n1 = ScalarParameter() {bUseCustomPrimitiveData=true, parameterName=OffsetAmount, primitiveDataIndex=4}
  n2 = VectorParameter() {bUseCustomPrimitiveData=true, parameterName=LookAtLocation, defaultValue=<rgba 0 0 0 1>}
  n3 = MaterialFunctionCall!(
   AnimationTexture=NamedRerouteUsage()
   AnimationTexture2=NamedRerouteUsage()
   AnimationTexture3=NamedRerouteUsage()
   AnimationState=NamedRerouteUsage()
   u=NamedRerouteUsage()
   bind_u=NamedRerouteUsage()
   texelHeight=NamedRerouteUsage()
   vStart=Constant()
   vHeight=Constant() {r=1}
   WorldNormal=NamedRerouteUsage()
  ) {materialFunction=/SparksMaps/Crowd/Materials/MF_Pilgrim_WPOTransform_subuvs}
  Add(
   A=NamedRerouteUsage()
   B=NamedRerouteDeclaration(None=MaterialFunctionCall!(Offset Amount=@n1, Normal, LookAtLocation=@n2.RGB, WPO=@n3.WPO) {materialFunction=/SparksMaps/Crowd/Materials/MF_Sparks_Crowd_Rotate_LookAt}.WPO) {name=AnimWPO, nodeColor=<rgba 1 0 0.3516 1>}
  )
  CollectionParameter() {collection=/SparksEffects/Effects/MPC_Sparks, desc="0 = Animation 1\r\n1 = Animation 2\r\n2 = Animation 3", parameterName=CrowdResponse}
  n4 = TextureCoordinate()
  n5 = AppendVector(A=OneMinus(None=ComponentMask(None=@n4) {r=true}), B=ComponentMask(None=@n4) {g=true})
  ComponentMask(None=@n5) {g=true}
  Constant()
  LinearInterpolate(
   A=Constant3Vector() {constant=<rgba 1 0 0.0788 0>}
   B=Constant3Vector() {constant=<rgba 0.1633 1 0 0>}
   Alpha=ComponentMask(None=TextureCoordinate() {coordinateIndex=1}) {r=true}
  )
  MaterialFunctionCall() {materialFunction=&TangentBasis}
  n6 = TextureCoordinate()
  n7 = CurveAtlasRowParameter(CurveTime=OneMinus(None=ComponentMask(None=@n6) {g=true})) {
   atlas=/SparksMaps/Crowd/Materials/LastLOD/Textures/CA_Sparks_CrowdLOD
   curve=/SparksMaps/Crowd/Materials/LastLOD/Textures/C_Sparks_Crowd_RedShirt
   defaultValue=7
  }
  MaterialFunctionCall!(BaseColor=@n7, Amount=Constant() {r=0.3}) {materialFunction=/Engine/Functions/Engine_MaterialFunctions01/SpeedTree/SpeedTreeColorVariation}
  n8 = ScalarParameter() {defaultValue=-5.0794, parameterName=Width}
  n9 = WorldPosition() {worldPositionShaderOffset=WPT_ExcludeAllShaderOffsets}
  n10 = NamedRerouteUsage()
  n11 = MaterialFunctionCall(
   WidthBase=@n8
   WidthTip=@n8
   WorldPosition=ComponentMask(None=Add(A=@n9.XYZ, B=@n10)) {b=true, g=true, r=true}
   UVs for Projection=@n5
   Expand U or V UV Channel=StaticBool() {value=true}
  ) {materialFunction=&SplineThicken}
  Multiply(A=@n11.Normal) {constB=-1}
  n12 = Frac(None=ComponentMask(None=TextureCoordinate()) {r=true})
  n13 = AppendVector(A=ConstantBiasScale(None=OneMinus(None=@n12)) {bias=-0.5, scale=2}, B=Constant())
  n14 = Subtract(A=CameraPositionWS(), B=WorldPosition().XYZ)
  n15 = Normalize(VectorInput=Transform(None=Constant3Vector() {constant=<rgba 0 1 0 0>}) {transformType=TRANSFORM_Instance})
  n16 = Transform(None=MaterialFunctionCall!(Vector=@n15, Quat=NamedRerouteUsage(), Invert Quaternion) {materialFunction=/Engine/Functions/Engine_MaterialFunctions03/Math/MultiplyVectorWithQuaternion}.Result) {transformSourceType=TRANSFORMSOURCE_Instance}
  n17 = MaterialFunctionCall!(Vector1=Normalize(VectorInput=CrossProduct(A=@n14, B=@n16)), Vector2=Multiply(A=@n16) {constB=-1}) {materialFunction=/Engine/Functions/Engine_MaterialFunctions02/Math/CreateThirdOrthogonalVector}
  NamedRerouteDeclaration(None=MaterialFunctionCall(VectorToTransform=DeriveNormalZ(InXY=@n13), BasisX=@n17.Vector1, BasisY=@n17.Vector1, BasisZ=@n17.Vector1) {materialFunction=&Transform3x3Matrix}.Transformed Vector) {name=Normal, nodeColor=<rgba 0.8516 1 0 1>}
  n18 = TextureObjectParameter() {
   parameterName=Animation Texture
   samplerType=SAMPLERTYPE_LinearColor
   texture=/SparksMaps/Crowd/Textures/SoldierMannequin_Male_GalileoShow_Gesture
  }
  n19 = TextureProperty(Tex=NamedRerouteDeclaration(None=@n18) {name=AnimationTexture, nodeColor=<rgba 0 1 0.1328 1>})
  n20 = TextureObjectParameter() {parameterName=Animation Texture 2, samplerType=SAMPLERTYPE_LinearColor, texture=/SparksMaps/Crowd/Textures/T_Sparks_GalileoShow_Cheer_Male}
  n21 = TextureProperty(Tex=NamedRerouteDeclaration(None=@n20) {name=AnimationTexture2, nodeColor=<rgba 0.5703 1 0 1>})
  n22 = MaterialFunctionCall!(Crowd Response=NamedRerouteDeclaration(None=CollectionParameter() {collection=&MPC_Sparks, parameterName=CrowdResponse}) {name=CrowdResponse, nodeColor=<rgba 0 1 0.5781 1>}) {materialFunction=/SparksMaps/Crowd/Materials/MF_Sparks_CrowdResponse_Blend_Offset}
  n23 = Round(None=@n22.Lerp Alpha 1)
  n24 = TextureObjectParameter() {
   parameterName=Animation Texture 3
   samplerType=SAMPLERTYPE_LinearColor
   texture=/SparksMaps/Crowd/Textures/T_Sparks_PapayaComms_Cheer_CMM_Male
  }
  n25 = NamedRerouteDeclaration(None=@n24) {name=AnimationTexture3, nodeColor=<rgba 0 0.3438 1 1>}
  n26 = MaterialFunctionCall(Float2=LinearInterpolate(A=LinearInterpolate(A=@n19, B=@n21, Alpha=@n23), B=TextureProperty(Tex=@n25), Alpha=Round(None=@n22.Lerp Alpha 2))) {materialFunction=&BreakOutFloat2Components}
  n27 = Divide(B=NamedRerouteDeclaration(None=@n26.R) {name=textureWidth, nodeColor=<rgba 0.0078 0 1 1>})
  NamedRerouteDeclaration(None=Multiply(A=NamedRerouteDeclaration(None=@n27) {name=texelWidth, nodeColor=<rgba 0 1 0.2266 1>}) {constB=0.5}) {name=halfTexelWidth, nodeColor=<rgba 0.0312 0 1 1>}
  n28 = ObjectPositionWS()
  n29 = Add(A=ComponentMask(None=@n28) {r=true}, B=ComponentMask(None=@n28) {g=true})
  n30 = Add(A=Add(A=Time(), B=Divide(A=@n29) {constB=100}), B=PerInstanceRandom())
  n31 = Divide(B=NamedRerouteUsage()) {constA=30}
  n32 = Frac(None=Add(A=Multiply(A=@n30, B=@n31)) {constB=0})
  n33 = NamedRerouteUsage()
  NamedRerouteDeclaration(None=Add(A=Multiply(A=@n32, B=@n33), B=NamedRerouteUsage())) {name=u, nodeColor=<rgba 0 1 0.9766 1>}
  NamedRerouteDeclaration(None=Add(A=NamedRerouteUsage(), B=NamedRerouteUsage())) {name=uStart, nodeColor=<rgba 1 0.3516 0 1>}
  n34 = NamedRerouteUsage()
  NamedRerouteDeclaration(None=Subtract(A=Subtract(B=@n34), B=@n34)) {name=uWidth, nodeColor=<rgba 0.9922 1 0 1>}
  NamedRerouteDeclaration(None=Divide(B=NamedRerouteDeclaration(None=@n26.G) {name=textureHeight, nodeColor=<rgba 0.5469 0 1 1>})) {name=texelHeight, nodeColor=<rgba 0 1 0.2266 1>}
  n35 = NamedRerouteUsage()
  n36 = CrossProduct(
   A=@n14
   B=Transform(None=MaterialFunctionCall(Vector=@n15, Quat=@n35) {materialFunction=&MultiplyVectorWithQuaternion}.Result) {transformSourceType=TRANSFORMSOURCE_Instance}
  )
  n37 = Multiply(A=@n8, B=NamedRerouteUsage())
  n38 = ComponentMask(None=Add(A=Constant() {r=-0.5}, B=@n5)) {r=true}
  NamedRerouteDeclaration(None=Multiply(A=Normalize(VectorInput=@n36), B=Multiply(A=@n37, B=@n38))) {name=CheapSplineThicken, nodeColor=<rgba 0 1 0.8359 1>}
  NamedRerouteDeclaration(None=@n3.QuaternionVS) {name=Quat, nodeColor=<rgba 0 0.4375 1 1>}
  NamedRerouteDeclaration(None=@n3.Quaternion) {name=QuatPS, nodeColor=<rgba 0 1 0.7891 1>}
  NamedRerouteDeclaration(None=@n7.A) {name=CurveAlpha, nodeColor=<rgba 1 0.8438 0 1>}
  n39 = Multiply(
   A=OneMinus(None=VertexColor().A)
   B=StaticSwitchParameter(
    True=MaterialFunctionCall!(CustomUPosition) {materialFunction=/SparksStageShow/Systems/StageManager/FXManager/MaterialFunctions/MF_MojoStageColors}.Primary
    False=VectorParameter() {group=Emissive, parameterName=Emissive Color, defaultValue=<rgba 0 1 0.3553 1>}.RGB
   ) {defaultValue=true, parameterName=Use Stage Colors for Emissive}
  )
  n40 = ScalarParameter() {defaultValue=100, group=Emissive, parameterName=Emissive Intensity}
  NamedRerouteDeclaration(None=StaticSwitchParameter(True=Multiply(A=@n39, B=@n40) {constB=100}, False=Constant()) {parameterName=Use Emissive}) {name=Emissive, nodeColor=<rgba 0 0.7187 1 1>}
  NamedRerouteUsage()
  NamedRerouteUsage()
  PreSkinnedPosition()
  ScalarParameter() {defaultValue=1, parameterName=Roughness}
  ScalarParameter() {parameterName=CrowdResponse, primitiveDataIndex=5}
  StaticBool() {value=true}
  TextureCoordinate()
  TextureCoordinate()
  VectorParameter() {parameterName=BaseColor, defaultValue=<rgba 0.2865 0.2865 0.2865 1>}
  VertexColor()
  VertexInterpolator(VS=@n11.Normal)
 ]
}
